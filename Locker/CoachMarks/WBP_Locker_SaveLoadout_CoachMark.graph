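# Locker coach mark, a sidebar onboarding tooltip that introduces Saved Loadouts for linking presets, with Skip and Next buttons.
# It wraps a modal UIKit coach mark, activated on show.
# Skip notifies close, deactivates both widgets and broadcasts OnDismissActionClicked.
# Next runs HandleOtherActionClicked, then shares a close path with tooltip end and focus-path removal, so all three broadcast OnOtherActionClicked.
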WidgetBlueprint /Locker/CoachMarks/WBP_Locker_SaveLoadout_CoachMark
graphs [
 EventGraph {nodes [
   "Button Dismiss Clicked (WBP_UIKit_CoachMark)"() {is=title K2Node_ComponentBoundEvent}
   > NotifyCloseAnimationFinished(self:Fort Sidebar Onboard Tooltip Widget Object Reference)
   > DeactivateWidget(self=Get WBP_UIKit_CoachMark() -> (WBP_UIKit_CoachMark:WBP UIKit Coach Mark Object Reference))
   > DeactivateWidget()
   > Call On Dismiss Action Clicked!(execute:Exec, self:WBP Locker Save Loadout Coach Mark Object Reference) -> (then:Exec) {is=title K2Node_CallDelegate}
   "Button Other Action Clicked (WBP_UIKit_CoachMark)"() {is=title K2Node_ComponentBoundEvent}
   > n1 = HandleOtherActionClicked(self:Fort Sidebar Onboard Tooltip Widget Object Reference)
   NotifyCloseAnimationFinished(execute=[@n1.then, @n2.then, @n3.then])
   > DeactivateWidget(self=Get WBP_UIKit_CoachMark())
   > DeactivateWidget()
   > Call On Other Action Clicked!(execute:Exec, self:WBP Locker Save Loadout Coach Mark Object Reference) -> (then:Exec) {is=title K2Node_CallDelegate}
   n2 = Event OnEndTooltip() {is=title K2Node_Event}
   n3 = Event OnRemovedFromFocusPath!() -> (OutputDelegate, then, InFocusEvent:Focus Event Structure) {is=title K2Node_Event}
   Event OnShowTooltip() {is=title K2Node_Event}
   > ActivateWidget(self=Get WBP_UIKit_CoachMark())
  ]}
 OnDismissActionClicked {nodes [OnDismissActionClicked() {is=title K2Node_FunctionEntry}]}
 OnOtherActionClicked {nodes [OnOtherActionClicked() {is=title K2Node_FunctionEntry}]}
]
variables [
 OnDismissActionClicked {}
 OnOtherActionClicked {}
]
widgets {
 info {
  inheritedWidgetCount=2
  parentClass=/Script/FortniteUI.FortSidebarOnboardTooltipWidget
  rootWidgetClass=/FortUILibrary/UIKit/CoachMark/WBP_UIKit_CoachMark.WBP_UIKit_CoachMark_C
  widgetCount=3
 }
 widgets [
  WBP_UIKit_CoachMark {
   widgetClassPath=/FortUILibrary/UIKit/CoachMark/WBP_UIKit_CoachMark
   bIsVariable=true
   widget=WBP_UIKit_CoachMark
   properties {
    bIsFocusable=true
    bIsModal=true
    bSupportsActivationFocus=false
    buttonFillRule=Automatic
    textDescription="Now you can link your presets together in one Saved Loadout!"
    textDismissButton=Skip
    textHeader=Introducing Saved Loadouts
    textOtherActionButton=Next
    visibility=SelfHitTestInvisible
    widget Width=650
    input Action Dismiss {dataTable=&InputActionDataTable, rowName=GenericBack}
    input Action Other {dataTable=&InputActionDataTable, rowName=GenericForward}
   }
  }
  Dismiss_Button {widgetClassPath=&CommonButtonLegacy, bInherited=true, widget=null}
  OtherAction_Button {widgetClassPath=&CommonButtonLegacy, bInherited=true, widget=null}
 ]
}
